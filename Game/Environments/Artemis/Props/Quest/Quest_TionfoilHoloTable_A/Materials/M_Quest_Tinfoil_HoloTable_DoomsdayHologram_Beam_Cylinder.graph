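# Unlit translucent material for the beam cylinder of the Doomsday hologram on the tinfoil holo-table quest prop.
# A UV-and-time sine drives color hue shift and axis-aligned Fresnel exponent.
# Opacity combines that Fresnel, a reflected UV gradient, scrolling world-height scanlines and summed-sine flicker evaluated per vertex.
# Flare and Blue scalar parameters provide blend alphas for flare variants of color, oscillation, Fresnel and normal-based vertex push, plus a blue tint.
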Material /Game/Environments/Artemis/Props/Quest/Quest_TionfoilHoloTable_A/Materials/M_Quest_Tinfoil_HoloTable_DoomsdayHologram_Beam_Cylinder
graph {
 properties {blendMode=BLEND_Translucent, pixelDepthOffsetMode=PDOM_Legacy, shadingModel=MSM_Unlit, usedShadingModels=Unlit}
 nodes [
  n1 = Multiply(
   A=ComponentMask(None=TextureCoordinate()) {g=true}
   B=LinearInterpolate(
    A=ScalarParameter() {parameterName=Oscillation Size}
    B=ScalarParameter() {defaultValue=1.6, parameterName=Oscillation Size Flare}
    Alpha=NamedRerouteUsage()
   )
  )
  n2 = Multiply(A=Time(), B=ScalarParameter() {defaultValue=4.23, parameterName=Oscillation Speed})
  n3 = ConstantBiasScale(None=Sine(None=Add(A=@n1, B=@n2)))
  n4 = VectorParameter() {parameterName=Beam Color, defaultValue=<rgba 9 1.1295 0.09 -0.015>}
  n5 = Multiply(A=@n3, B=@n4.A)
  n6 = VectorParameter() {parameterName=Beam Color Flare, defaultValue=<rgba 33 6.3195 0.33 -0.03>}
  n7 = Multiply(A=@n3, B=@n6.A)
  LinearInterpolate(
   A=LinearInterpolate(
    A=MaterialFunctionCall(Hue Shift Percentage=@n5, Texture=@n4.RGB) {materialFunction=&HueShift}.Result
    B=MaterialFunctionCall(Hue Shift Percentage=@n7, Texture=@n6.RGB) {materialFunction=&HueShift}.Result
    Alpha=NamedRerouteUsage()
   )
   B=VectorParameter() {parameterName=Beam Color Blue Flare, defaultValue=<rgba 0.03 2.0595 3 -0.03>}.RGB
   Alpha=NamedRerouteUsage()
  )
  Multiply(
   A=VertexNormalWS()
   B=LinearInterpolate(
    A=ScalarParameter() {defaultValue=-0.75, parameterName=Push}
    B=ScalarParameter() {defaultValue=1.5, parameterName=Push Flare}
    Alpha=NamedRerouteUsage()
   )
  )
  n8 = Power(Base=NamedRerouteUsage(), Exp=ScalarParameter() {defaultValue=2, parameterName=Scanlines Power})
  n9 = Multiply(A=@n8, B=NamedRerouteUsage())
  n10 = ScalarParameter() {defaultValue=2, parameterName=Axis Fresnel Exponent Min}
  n11 = ScalarParameter() {defaultValue=4, parameterName=Axis Fresnel Exponent Max}
  n12 = ScalarParameter() {defaultValue=1, parameterName=Axis Fresnel Exponent Min Flare}
  n13 = ScalarParameter() {defaultValue=22, parameterName=Axis Fresnel Exponent Max Flare}
  n14 = MaterialFunctionCall(Exponent=LinearInterpolate(A=LinearInterpolate(A=@n10, B=@n11, Alpha=@n3), B=LinearInterpolate(A=@n12, B=@n13, Alpha=@n3), Alpha=NamedRerouteUsage())) {materialFunction=&AxisAlignedFresnel}
  n15 = MaterialFunctionCall!(
   Gradient=ComponentMask(None=TextureCoordinate()) {g=true}
   Center Point=ScalarParameter() {defaultValue=0.85, parameterName=Gradient Center Point}
  ) {materialFunction=/Game/Effects/Fort_Effects/MaterialFunctions/MF_GradientReflected}
  n16 = Multiply(A=@n15.Result, B=@n15.Result)
  Multiply(A=Multiply(A=Add(A=@n8, B=@n9), B=Multiply(A=@n14.Fresnel, B=@n16)), B=ScalarParameter() {defaultValue=1, parameterName=Opacity})
  n17 = Time()
  n18 = ScalarParameter() {defaultValue=4, parameterName=ScanlineSpeed}
  n19 = MaterialFunctionCall(Float3=Add(A=WorldPosition().XYZ, B=Multiply(A=@n17, B=@n18))) {materialFunction=&BreakOutFloat3Components}
  n20 = ScalarParameter() {defaultValue=1.2, parameterName=ScanlineDivisor}
  n21 = SphereMask(
   A=Frac(None=Divide(A=@n19.B, B=@n20))
   B=Constant() {r=0.5}
   Radius=ScalarParameter() {defaultValue=0.55, parameterName=ScanlineRadius}
   Hardness=ScalarParameter() {defaultValue=0.5, parameterName=ScanlineHardness}
  )
  n22 = ScalarParameter() {defaultValue=4, parameterName=ScanlineExponent}
  NamedRerouteDeclaration(None=LinearInterpolate(
    A=ScalarParameter() {defaultValue=0.6, parameterName=ScanlineLowValue}
    B=ScalarParameter() {defaultValue=1, parameterName=ScanlineHighValue}
    Alpha=Power(Base=@n21, Exp=@n22)
   )) {name=Scanlines, nodeColor=<rgba 1 0 0.6094 1>}
  NamedRerouteDeclaration(None=Saturate(None=ScalarParameter() {parameterName=Flare})) {name=Flare Alpha, nodeColor=<rgba 1 0.6 0 1>}
  n23 = Time()
  n24 = Add(A=Add(A=Sine(None=@n23) {period=0.2676}, B=Sine(None=@n23) {period=0.1766}), B=Sine(None=@n23) {period=0.3324})
  n25 = ScalarParameter() {defaultValue=0.07, parameterName=FlickerScale}
  NamedRerouteDeclaration(None=VertexInterpolator(VS=Multiply(A=@n24, B=@n25)).PS) {name=Flicker, nodeColor=<rgba 0 1 0.0156 1>}
  NamedRerouteDeclaration(None=Saturate(None=ScalarParameter() {parameterName=Blue})) {name=Blue, nodeColor=<rgba 0 0.1797 1 1>}
 ]
}
