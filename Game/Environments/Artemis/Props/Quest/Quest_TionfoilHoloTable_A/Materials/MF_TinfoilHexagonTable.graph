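# Material function for the animated hexagon surface of the Tinfoil holo table quest prop, outputting base color, emissive, alpha, sweep gradient and UVs.
# A tiled, panning packed hexagon texture supplies masks, with green as alpha and stepped modulo-time speeds staggering the pulses.
# Emissive sums colored edge pulses, chroma-curve pulses gated by a sine-driven bounding-box sweep, and a faint view-dependent chroma background.
# Curve atlas lookups provide the shifting chroma colors.
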MaterialFunction /Game/Environments/Artemis/Props/Quest/Quest_TionfoilHoloTable_A/Materials/MF_TinfoilHexagonTable
graph {nodes [
  FunctionOutput(None=Multiply(A=NamedRerouteUsage(), B=VectorParameter() {parameterName=HexagonColor, defaultValue=<rgba 0.1134 0.1644 0.8212 0>}.RGB)) {outputName=BaseColor}
  n1 = MaterialFunctionCall() {materialFunction=&BoundingBoxBased_0-1_UVW}
  n2 = Multiply(A=Sine(None=Time()) {period=12.3}) {constB=2}
  n3 = ScalarParameter() {defaultValue=0.5, parameterName=SweepRadius}
  n4 = ScalarParameter() {defaultValue=0.1, parameterName=SweepHardness}
  n5 = ComponentMask(None=TextureCoordinate() {uTiling=0.9}) {r=true}
  n6 = MaterialFunctionCall!(Time, Divisor=Constant() {r=11}, Mod) {materialFunction=/Game/Effects/Fort_Effects/Materials/_Functions/MF_ModuloTime}
  n7 = CurveAtlasRowParameter(CurveTime=Frac(None=Add(A=@n5, B=@n6.Clamped 0-1))) {
   atlas=/Game/Environments/Artemis/Props/Quest/Quest_TionfoilHoloTable_A/Textures/CA_Tinfoil_Holotable
   curve=/Game/Environments/Artemis/Props/Quest/Quest_TionfoilHoloTable_A/Textures/C_Tinfoil_Holotable_Chroma
   parameterName=EmissiveBGColorCurve
  }
  n8 = Frac(None=Add(A=NamedRerouteUsage(), B=NamedRerouteUsage()))
  n9 = Constant() {r=0.5}
  n10 = LinearInterpolate(
   A=FunctionInput(Preview=ScalarParameter() {defaultValue=0.2, parameterName=EmissiveEdgeRadiusLow}) {bUsePreviewValueAsDefault=true, inputName=EmissiveEdgeRadiusLow, inputType=FunctionInput_Scalar, sortPriority=4}
   B=FunctionInput(Preview=ScalarParameter() {defaultValue=0.7, parameterName=EmissiveEdgeRadiusHigh}) {bUsePreviewValueAsDefault=true, inputName=EmissiveEdgeRadiusHigh, inputType=FunctionInput_Scalar, sortPriority=5}
   Alpha=NamedRerouteUsage()
  )
  n11 = Constant()
  n12 = Multiply(
   A=NamedRerouteUsage()
   B=Power(Base=SphereMask(A=@n8, B=@n9, Radius=@n10, Hardness=@n11), Exp=ScalarParameter() {defaultValue=16, parameterName=EmissiveEdgePower})
  )
  n13 = Multiply(
   A=FunctionOutput(None=SphereMask(A=@n1.R, B=@n2, Radius=@n3, Hardness=@n4)) {outputName=SweepGradient, sortPriority=3}
   B=Multiply(A=Multiply(A=@n7, B=@n12), B=ScalarParameter() {defaultValue=200, parameterName=ChromaEdgeScale})
  )
  n14 = Multiply(
   A=@n12
   B=FunctionInput(Preview=VectorParameter() {parameterName=EmissiveColor, defaultValue=<rgba 4.1072 12.6475 50 0>}.RGB) {bUsePreviewValueAsDefault=true, inputName=EmissiveColor, sortPriority=6}
  )
  n15 = CameraVectorWS()
  n16 = ReflectionVectorWS()
  n17 = Add(A=Add(A=NamedRerouteUsage(), B=NamedRerouteUsage()), B=ConstantBiasScale(None=DotProduct(A=@n15, B=@n16)))
  n18 = Constant() {r=1}
  n19 = CurveAtlasRowParameter(CurveTime=SphereMask(A=Frac(None=@n17), B=@n18, Radius=@n18, Hardness=Constant())) {
   atlas=&CA_Tinfoil_Holotable
   curve=&C_Tinfoil_Holotable_Chroma
   parameterName=EmissiveBGColorCurve
  }
  n20 = Multiply(A=@n19, B=NamedRerouteUsage())
  n21 = ScalarParameter() {defaultValue=0.02, parameterName=EmissiveBGScale}
  FunctionOutput(None=Add(A=Add(A=@n13, B=@n14), B=Multiply(A=@n20, B=@n21))) {outputName=Emissive, sortPriority=1}
  n22 = FunctionInput(Preview=TextureCoordinate()) {bUsePreviewValueAsDefault=true, inputName=UVs, inputType=FunctionInput_Vector2, sortPriority=-1}
  n23 = FunctionInput(Preview=ScalarParameter() {defaultValue=1.6, parameterName=HoloTableTiling}) {bUsePreviewValueAsDefault=true, inputName=Tiling, inputType=FunctionInput_Scalar}
  n24 = AppendVector(
   A=ScalarParameter() {defaultValue=0.0133, parameterName=HoloTablePanSpeed_U}
   B=ScalarParameter() {defaultValue=0.01, parameterName=HoloTablePanSpeed_V}
  )
  n25 = NamedRerouteDeclaration(None=Panner(
    Coordinate=Multiply(A=@n22, B=@n23)
    Speed=FunctionInput(Preview=@n24) {bUsePreviewValueAsDefault=true, inputName=PanSpeed, inputType=FunctionInput_Vector2, sortPriority=1}
   ) {bFractionalPart=true}) {name=UVs, nodeColor=<rgba 0 1 0.4141 1>}
  n26 = TextureSample(UVs=@n25) {
   samplerType=SAMPLERTYPE_LinearColor
   texture=/Game/Environments/Artemis/Props/Quest/Quest_TionfoilHoloTable_A/Textures/T_HoloTable_Hexagons_A
  }
  FunctionOutput(None=NamedRerouteDeclaration(None=@n26.G) {name=Green, nodeColor=<rgba 0 1 0 1>}) {outputName=Alpha, sortPriority=2}
  FunctionOutput(None=@n25) {outputName=UVs, sortPriority=4}
  LinearInterpolate(
   A=ScalarParameter() {defaultValue=1, parameterName=EmissiveBGRadiusLow}
   B=ScalarParameter() {defaultValue=2, parameterName=EmissiveBGRadiusHigh}
  )
  NamedRerouteDeclaration(None=Power(Base=@n26.R)) {name=Red, nodeColor=<rgba 1 0 0.8437 1>}
  NamedRerouteDeclaration(None=@n26.B) {name=Blue, nodeColor=<rgba 0 0 1 1>}
  n27 = FunctionInput(Preview=ScalarParameter() {defaultValue=2, parameterName=EmissiveEdgeSpeedLow}) {bUsePreviewValueAsDefault=true, inputName=EmissiveEdgeSpeedLow, inputType=FunctionInput_Scalar, sortPriority=2}
  n28 = FunctionInput(Preview=ScalarParameter() {defaultValue=8, parameterName=EmissiveEdgeSpeedHigh}) {bUsePreviewValueAsDefault=true, inputName=EmissiveEdgeSpeedHigh, inputType=FunctionInput_Scalar, sortPriority=3}
  n29 = NamedRerouteUsage()
  n30 = Constant() {r=6}
  n31 = Divide(A=Round(None=Multiply(A=@n29, B=@n30)), B=@n30)
  NamedRerouteDeclaration(None=MaterialFunctionCall(Divisor=LinearInterpolate(A=@n27, B=@n28, Alpha=@n31) {constA=1, constB=4}) {materialFunction=&MF_ModuloTime}.Unclamped) {name=Time, nodeColor=<rgba 1 0.3984 0 1>}
  NamedRerouteDeclaration(None=@n26.A) {name=Alpha, nodeColor=<rgba 1 1 1 1>}
 ]}
