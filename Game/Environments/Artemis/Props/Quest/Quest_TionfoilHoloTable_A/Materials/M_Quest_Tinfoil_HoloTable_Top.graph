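# Opaque lit material for the top of the Tinfoil quest holo table, showing animated glowing hexagons on a glossy, metallic blue surface.
# Two calls to MF_TinfoilHexagonTable form a main layer and a dimmed, bump-offset panning second layer, feeding base color and emissive.
# Their combined alpha masks a parallax background hex pattern tinted by the sweep gradient, with a textured edge glow added on top.
# Normals blend two panning cloud maps with hexagon normals.
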Material /Game/Environments/Artemis/Props/Quest/Quest_TionfoilHoloTable_A/Materials/M_Quest_Tinfoil_HoloTable_Top
graph {
 properties {bUsedWithNanite=true, pixelDepthOffsetMode=PDOM_Legacy, usedShadingModels=Default Lit}
 nodes [
  n1 = SphereMask(
   A=TextureCoordinate()
   B=Constant() {r=0.5}
   Radius=ScalarParameter() {defaultValue=0.6, parameterName=BaseColorRadius}
   Hardness=ScalarParameter() {defaultValue=0.2, parameterName=BaseColorHardness}
  )
  n2 = ScalarParameter() {defaultValue=3, parameterName=BaseColorExponent}
  n3 = ScalarParameter() {defaultValue=0.1, parameterName=2ndLayerScale}
  n4 = MaterialFunctionCall!(
   UVs=BumpOffset(Coordinate=TextureCoordinate(), Height=ScalarParameter() {defaultValue=-1, parameterName=2ndLayerHeight})
   Tiling=AppendVector(A=ScalarParameter() {defaultValue=-1.4, parameterName=2ndLayer_U}, B=ScalarParameter() {defaultValue=-1.4, parameterName=2ndLayer_V})
   PanSpeed=AppendVector(
    A=ScalarParameter() {defaultValue=-0.002, parameterName=2ndLayer_PanSpeedU}
    B=ScalarParameter() {defaultValue=-0.0018, parameterName=2ndLayer_PanSpeedV}
   )
   EmissiveEdgeSpeedLow
   EmissiveEdgeSpeedHigh
   EmissiveEdgeRadiusLow
   EmissiveEdgeRadiusHigh
   EmissiveColor
  ) {materialFunction=&/MF_TinfoilHexagonTable}
  n5 = MaterialFunctionCall() {materialFunction=&MF_TinfoilHexagonTable}
  Add(
   A=LinearInterpolate(
    A=VectorParameter() {parameterName=BaseColor, defaultValue=<rgba 0.0205 0.0317 0.1545 0>}.RGB
    B=VectorParameter() {parameterName=BaseColorEdges, defaultValue=<rgba 0.0362 0.4005 2 0>}.RGB
    Alpha=Power(Base=@n1, Exp=@n2)
   )
   B=LinearInterpolate(A=Multiply(A=@n3, B=@n4.BaseColor), B=@n5.BaseColor, Alpha=@n5.BaseColor)
  )
  n6 = Multiply(A=@n4.Emissive, B=@n3)
  n7 = OneMinus(None=NamedRerouteDeclaration(None=Max(A=@n5.Alpha, B=@n4.Alpha)) {name=HexMasks, nodeColor=<rgba 0.4062 1 0 1>})
  n8 = TextureCoordinate()
  n9 = ScalarParameter() {defaultValue=-2, parameterName=BGHexHeight}
  n10 = TextureSample(UVs=Multiply(A=BumpOffset(Coordinate=@n8, Height=@n9), B=ScalarParameter() {defaultValue=4, parameterName=BGHexScale_UV})) {
   samplerType=SAMPLERTYPE_LinearColor
   texture=/Game/Environments/Artemis/Props/Quest/Quest_TionfoilHoloTable_A/Textures/T_Tinfoil_HoloTable_Hexagons
  }
  n11 = Multiply(
   A=@n10.R
   B=LinearInterpolate(
    A=VectorParameter() {parameterName=BGHexColor, defaultValue=<rgba 0.0112 0.0436 0.0469 0>}.RGB
    B=VectorParameter() {parameterName=BGHexColorB, defaultValue=<rgba 0.2292 0.0356 0.2126 0>}.RGB
    Alpha=@n5.SweepGradient
   )
  )
  Add(
   A=Add(A=LinearInterpolate(A=@n6, B=@n5.Emissive, Alpha=@n5.Emissive), B=Multiply(A=@n7, B=@n11))
   B=Multiply(
    A=TextureSample() {texture=/Game/Environments/Artemis/Props/Quest/Quest_TionfoilHoloTable_A/Textures/T_Holotable_EdgeGlow}.R
    B=VectorParameter() {parameterName=EdgeGlow, defaultValue=<rgba 0.0342 0.1746 1 0>}.RGB
   )
  )
  Constant3Vector() {constant=<rgba 0 0 1 0>}
  n12 = VectorParameter() {parameterName=NormapTilePanA, defaultValue=<rgba 0.366 0.366 0.012 0.0106>}
  n13 = Panner(Coordinate=Multiply(A=TextureCoordinate(), B=AppendVector(A=@n12.R, B=@n12.R)), Speed=AppendVector(A=@n12.B, B=@n12.B)) {speedX=0.2, speedY=0.2}
  n14 = VectorParameter() {parameterName=NormapTilePanB, defaultValue=<rgba -0.286 -0.286 0.0096 0.012>}
  n15 = Panner(Coordinate=Multiply(A=TextureCoordinate(), B=AppendVector(A=@n14.R, B=@n14.R)), Speed=AppendVector(A=@n14.B, B=@n14.B)) {speedX=-0.115, speedY=-0.081}
  n16 = Multiply(
   A=Add(
    A=TextureSample(UVs=@n13) {samplerType=SAMPLERTYPE_Normal, texture=/Game/Effects/Fort_Effects/Creatures/Elemental/Ice/TilingCloudNormalsMap}.RGB
    B=TextureSample(UVs=@n15) {samplerType=SAMPLERTYPE_Normal, texture=&TilingCloudNormalsMap}.RGB
   )
   B=VectorParameter() {parameterName=NormalStrength, defaultValue=<rgba 0.3 0.3 1 0>}.RGB
  )
  n17 = TextureSample(UVs=NamedRerouteUsage()) {samplerType=SAMPLERTYPE_Normal, texture=/Game/Environments/Artemis/Props/Quest/Quest_TionfoilHoloTable_A/Textures/T_Hexagons_N}
  n18 = NamedRerouteUsage()
  NamedRerouteDeclaration(None=Normalize(VectorInput=LinearInterpolate(A=@n16, B=@n17.RGB, Alpha=@n18))) {name=Normals, nodeColor=<rgba 1 0.1406 0 1>}
  NamedRerouteDeclaration(None=@n5.UVs) {name=Hex_UVs, nodeColor=<rgba 0.875 1 0 1>}
  NamedRerouteUsage()
  ScalarParameter() {defaultValue=0.6, parameterName=Metallic}
  ScalarParameter() {defaultValue=0.2, parameterName=Specular}
  ScalarParameter() {defaultValue=0.12, parameterName=Roughness}
 ]
}
